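; Quest ability that credits the owning player for surviving a high fall without taking fall damage.
; On event activation during Athena.GameOn, it records the pawn's starting Z, then waits for the falling movement tag to be removed.
; If the pawn is alive and has dropped at least the height threshold, it broadcasts a HighFall custom verb message.
; A parallel wait ends the ability early if a fall-damage event arrives.
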(event Ability|EventActivateAbilityFromEvent
  (params (EventData :type Gameplay%20Event%20Data%20Structure%20%28by%20ref%29))
  (bind (_outsourcetags _outcontexttags) (Quests|GetSourceandContextTags (Variables|GAVars|Data|GetOwningPlayerQuestManager)))
  (|CallParentFunction EventData)
  (if (GameplayTags|HasTag _outcontexttags "(TagName=\"Athena.GameOn\")")
    (Variables|Default|SetActor%20Initial%20Height (.z (Transformation|GetActorLocation (Variables|GAVars|Data|GetOwningPlayerPawn))))
    (Ability|Tasks|WaitGameplayEvent "(TagName=\"Event.Damage.Received.Falling\")"
      (:then
        (Ability|Tasks|WaitGameplayTagRemove "(TagName=\"MovementMode.Falling\")"
          (:Removed
            (if (FortPawn|IsDead (Variables|GAVars|Data|GetOwningPlayerPawn))
              _
              (elif (>= (- (Variables|Default|GetActor%20Initial%20Height) (Variables|Default|GetHeight%20Check)) (.z (Transformation|GetActorLocation (Variables|GAVars|Data|GetOwningPlayerPawn))))
                (CustomVerb|BroadcastCustomVerbMessage (CustomVerb|MakeCustomVerbMessage (Variables|GAVars|Data|GetOwningPlayerPawn) "(TagName=\"CustomVerb.HighFall\")"))
                (else
                  _)))
            (Ability|EndAbility))))
      (:EventReceived
        (Ability|EndAbility)))
    (else
      (Ability|EndAbility))))
